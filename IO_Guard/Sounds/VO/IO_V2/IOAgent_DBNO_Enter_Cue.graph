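# Sound cue for the IO Agent guard voice line played on entering the DBNO (down but not out) state.
# A Random node picks one of four equally weighted wave variations and is excluded from branch culling.
# It routes through the NPC VO sound class with a close-range guard attenuation, a chatter concurrency group and a slight volume boost.
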SoundCue /IO_Guard/Sounds/VO/IO_V2/IOAgent_DBNO_Enter_Cue
root=Random_1
nodes [
 {
  id=Random_1
  children [WavePlayer_0,WavePlayer_5,WavePlayer_6,WavePlayer_7]
  properties {bShouldExcludeFromBranchCulling=true, weights [1,1,1,1]}
 }
 {id=WavePlayer_0, properties {soundWaveAssetPtr=&/IOAgents_2021_DBNO_Enter_010}}
 {id=WavePlayer_5, properties {soundWaveAssetPtr=&/IOAgents_2021_DBNO_Enter_011}}
 {id=WavePlayer_6, properties {soundWaveAssetPtr=&/IOAgents_2021_DBNO_Enter_012}}
 {id=WavePlayer_7, properties {soundWaveAssetPtr=&/IOAgents_2021_DBNO_Enter_013}}
]
properties {
 attenuationSettings=/IO_Guard/Sounds/VO/IO_Guard_Close_Att
 maxDistance=4000
 soundClassObject=&Fort_VO_NPC
 volumeMultiplier=1.2
 concurrencySet [/IO_Guard/Sounds/VO/IO_Guard_Chatter_Conc]
}
