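# IO Guard voice cue that plays a full-alert idle line for IO agents.
# A Random node picks one of five equally weighted wave variations and is excluded from branch culling.
# It routes through the NPC VO sound class, using the guard idle attenuation and a chatter concurrency set to govern overlapping barks.
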SoundCue /IO_Guard/Sounds/VO/IO_V2/IOAgent_FullAlertIdle_Cue
root=Random_0
nodes [
 {
  id=Random_0
  children [WavePlayer_1,WavePlayer_2,WavePlayer_3,WavePlayer_4,WavePlayer_5]
  properties {bShouldExcludeFromBranchCulling=true, weights [1,1,1,1,1]}
 }
 {id=WavePlayer_1, properties {soundWaveAssetPtr=&/IOAgents_2021_FullAlertIdle_010}}
 {id=WavePlayer_2, properties {soundWaveAssetPtr=&/IOAgents_2021_FullAlertIdle_020}}
 {id=WavePlayer_3, properties {soundWaveAssetPtr=&/IOAgents_2021_FullAlertIdle_030}}
 {id=WavePlayer_4, properties {soundWaveAssetPtr=&/IOAgents_2021_FullAlertIdle_040}}
 {id=WavePlayer_5, properties {soundWaveAssetPtr=&/IOAgents_2021_FullAlertIdle_050}}
]
properties {
 attenuationSettings=/IO_Guard/Sounds/VO/IO_Guard_Idle_Att
 maxDistance=4000
 soundClassObject=&Fort_VO_NPC
 volumeMultiplier=1
 concurrencySet [/IO_Guard/Sounds/VO/IO_Guard_Chatter_Conc]
}
