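# Sound cue for an IO Guard tank driver's voice lines about repairing a tread.
# A random node picks one of five equally weighted wave recordings each time it plays.
# It plays as NPC voice-over at boosted volume, sharing the IO Guard idle attenuation and chatter concurrency settings to limit distance and overlapping barks.
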SoundCue /IO_Guard/Sounds/VO/IO_V2/IOGuard_TankDriver_2022_RepairTread_Cue
root=Random_0
nodes [
 {id=Random_0, children [WavePlayer_0,WavePlayer_1,WavePlayer_2,WavePlayer_3,WavePlayer_4], properties {weights [1,1,1,1,1]}}
 {id=WavePlayer_0, properties {soundWaveAssetPtr=&/IOGuard_TankDriver_2022_RepairTread_010}}
 {id=WavePlayer_1, properties {soundWaveAssetPtr=&/IOGuard_TankDriver_2022_RepairTread_020}}
 {id=WavePlayer_2, properties {soundWaveAssetPtr=&/IOGuard_TankDriver_2022_RepairTread_030}}
 {id=WavePlayer_3, properties {soundWaveAssetPtr=&/IOGuard_TankDriver_2022_RepairTread_040}}
 {id=WavePlayer_4, properties {soundWaveAssetPtr=&/IOGuard_TankDriver_2022_RepairTread_050}}
]
properties {
 attenuationSettings=/IO_Guard/Sounds/VO/IO_Guard_Idle_Att
 maxDistance=4000
 soundClassObject=&Fort_VO_NPC
 volumeMultiplier=1.5
 concurrencySet [/IO_Guard/Sounds/VO/IO_Guard_Chatter_Conc]
}
